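; Player pawn component that spawns and manages the Sprite backpack actor, coordinating its visibility with the pawn's back bling.
; CVar-gated spawning runs locally or through replicated class and weapon GUID data, whose OnRep rebuilds the backpack on non-autonomous clients.
; Weapon equips, Sprite item-visibility events and disable or force-hide tags drive hiding.
; Hidden changes update the actor, server state and back bling part, then broadcast HiddenChanged.
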(event EventBeginPlay
  (|CallParentFunction)
  (CallFunction|SetServerHiddenStatus true)
  (bind _asfort_player_pawn_athena (Utilities|Casting|CastToFortPlayerPawnAthena (Components|GetOwner)))
  (bind _output_get (Variables|Default|SetOwnerPawn _asfort_player_pawn_athena))
  (FortPawn|BindEventtoOnWeaponEquippedDelegate _output_get (delegate Custom|OnWeaponEquipped))
  (bind _asynctaskproxy (Ability|Tasks|WaitGameplayTagCountChangedOnActor (Variables|Default|GetOwnerPawn) (Variables|Default|GetTagDisableBackpack))
    (:then
      (Variables|Default|SetDisableBackpackAction _asynctaskproxy))
    (:TagCountChanged
      (CallFunction|RefreshBackpackVisibility))))

(event Custom|OnWeaponEquipped
  (params (NewWeapon :type Fort%20Weapon%20Object%20Reference) (PrevWeapon :type Fort%20Weapon%20Object%20Reference))
  (Variables|Default|SetAllowHideBackCosmetic (CallFunction|AllowHideBackbling (Variables|Default|GetOwnerPawn)))
  (CallFunction|RefreshBackpackVisibility))

(event EventEndPlay
  (params (EndPlayReason :type EEndPlayReason%20Enum :default "Destroyed"))
  (|CallParentFunction EndPlayReason)
  (bind _backpackactorref (Variables|Default|GetBackpackActorRef)
    (:then
      (Actor|DestroyActor _backpackactorref)
      (AsyncAction|Cancel (Variables|Default|GetDisableBackpackAction))
      (FortPawn|UnbindEventfromOnWeaponEquippedDelegate (Variables|Default|GetOwnerPawn) (delegate Custom|OnWeaponEquipped)))
    (:else)))

(event Custom|DestroyBackpack
  (params (ItemGuid :type Guid%20Structure))
  (bind _backpackactorref (Variables|Default|GetBackpackActorRef))
  (if (== ItemGuid (CallFunction|GetCurrentLocalWeaponGuid))
    (CallFunction|SetLocalClassandGuid 0 (Utilities|Struct|MakeGuid))
    (CallFunction|ClearListenEvents)
    (CallFunction|SetBackpackHidden true)
    (Variables|Default|SetHidden true)
    (Utilities|IsValid _backpackactorref
      (:Is%20Valid
        (Actor|DestroyActor _backpackactorref)))))

(event Custom|RefreshBackpackVisibility
  ; The backpack sprite is hidden while the sprite is equipped, but also when a different item utilizes the backpack cosmetic slot
  (CallFunction|SetBackpackHidden (or (or (not (Utilities|IsValid (Variables|Default|GetBackpackActorRef) :!class /Script/Engine.KismetSystemLibrary)) (or (Variables|Default|GetSpriteWeaponVisible) (not (Variables|Default|GetAllowHideBackCosmetic)))) (GameplayTags|HasMatchingGameplayTag (Variables|Default|GetOwnerPawn) (Variables|Default|GetTagDisableBackpack)))))

(event Custom|SpawnBackpackServer
  (params (BackpackClass :type Actor%20Class%20Reference) (ItemGuid :type Guid%20Structure))
  ;[0.15 0.1448 0.06549 0.5] Server Functions
  (if (Default|IsBackpackReplicationEnabled)
    (CallFunction|SetClassandGuidforReplication BackpackClass ItemGuid)
    (CallFunction|BindEvents)))

(event Custom|ClearListenEvents
  ;[0.15 0.1448 0.06549 0.5] Server Functions
  (Ability|Async|EndAction (Variables|Default|GetVIsibilityAction))
  (Ability|Async|EndAction (Variables|Default|GetShrineOfferAction)))

(event Custom|ClearBackpackServer
  (params (ItemGUID :type Guid%20Structure))
  ;[0.15 0.1448 0.06549 0.5] Server Functions
  (if (== ItemGUID (CallFunction|GetCurrentReplicatedWeaponGuid))
    (CallFunction|SetClassandGuidforReplication 0 (Utilities|Struct|MakeGuid))
    (CallFunction|ClearListenEvents)
    (CallFunction|SetServerHiddenStatus true)))

(event Custom|SetBackpackHidden
  (params (NewHidden :type Boolean))
  (bind _backpackactorref (Variables|Default|GetBackpackActorRef))
  (bind _ownerpawn (Variables|Default|GetOwnerPawn))
  ;[0.15 0.04984 0.03487 0.5] Visibility Handling
  (if (!= NewHidden (Variables|Default|GetHidden))
    (bind _output_get (Variables|Default|SetHidden NewHidden))
    (Utilities|IsValid _backpackactorref
      (:Is%20Valid
        (Rendering|SetActorHiddenInGame _backpackactorref _output_get))
      (:Is%20Not%20Valid
        _))
    (CallFunction|SetServerHiddenStatus _output_get)
    (CallFunction|SetPartVisibility _ownerpawn _output_get)
    ; If Fort.Rollback.ForcePartVisibility
    ; Is 0, this will not do anything.
    ; Otherwise, it will add/remove enough until the desired visibility is reached
    (CallFunction|ForceBackblingPartVisibility _ownerpawn _output_get)
    (Default|CallHiddenChanged (Variables|Default|GetHidden))))

(event Custom|BindWeaponVisbilityEvent
  ;[0.15 0.04984 0.03487 0.5] Visibility Handling
  (bind _asfort_player_pawn_athena (Utilities|Casting|CastToFortPlayerPawnAthena (Components|GetOwner)))
  (bind (_asynctaskproxy _payload) (Ability|Async|WaitGameplayEventToActor _asfort_player_pawn_athena "(TagName=\"Athena.Player.Sprite.ItemVisibilityChanged\")")
    (:then
      (Variables|Default|SetVIsibilityAction _asynctaskproxy))
    (:EventReceived
      (Variables|Default|SetSpriteWeaponVisible (Widget|UEFN|ToBool(Double) (Utilities|Struct|BreakGameplayEventData _payload)))
      (CallFunction|RefreshBackpackVisibility))))

(event Custom|BindEvents
  ;[0.15 0.04984 0.03487 0.5] Visibility Handling
  (bind _asfort_player_pawn_athena (Utilities|Casting|CastToFortPlayerPawnAthena (Components|GetOwner)))
  (CallFunction|BindWeaponVisbilityEvent)
  (bind (_asynctaskproxy _tagcount) (Ability|Tasks|WaitGameplayTagCountChangedOnActor _asfort_player_pawn_athena "(TagName=\"Granted.Athena.Player.Sprite.Backpack.ForceHide\")")
    (:then
      (Variables|Default|SetShrineAction _asynctaskproxy))
    (:TagCountChanged
      (if (> _tagcount 0)
        (Ability|Async|EndAction (Variables|Default|GetVIsibilityAction))
        (CallFunction|SetBackpackHidden true)
        (else
          (CallFunction|BindWeaponVisbilityEvent))))))

(event EventHandleReceivedServerHiddenState
  (params (bServerHidden :type Boolean)))

(event Custom|InitializeBackpack
  ;[0.15 0.04984 0.03487 0.5] Visibility Handling
  (bind _asfort_player_pawn_athena (Utilities|Casting|CastToFortPlayerPawnAthena (Components|GetOwner)))
  (Variables|Default|SetOwnerPawn _asfort_player_pawn_athena)
  (Class|BSpriteHeldBackpack|InitializeforPawn (Variables|Default|GetBackpackActorRef) (Variables|Default|GetOwnerPawn))
  (CallFunction|BindEvents))

(event EventOnRep_ReplicatedBackpackData
  (bind (_outbackpackclass _returnvalue) (CallFunction|GetCurrentReplicatedBackpackClass))
  (bind (_outguid _returnvalue_2) (CallFunction|GetCurrentLocalWeaponGuid))
  ;[0.05898 0.06582 0.15 0.5] Simulated Promies Replication
  (if (Utilities|Enum|NotEqual(Enum) (Networking|GetLocalRole (Components|GetOwner)) "ROLE_AutonomousProxy")
    (if (Default|IsBackpackReplicationEnabled)
      (Networking|CanExecuteCosmeticEvents
        (:True
          (bind _asfort_player_pawn_athena (Utilities|Casting|CastToFortPlayerPawnAthena (Components|GetOwner)))
          (Variables|Default|SetOwnerPawn _asfort_player_pawn_athena)
          (Utilities|FlowControl|Sequence
            (:then_0
              (if _returnvalue
                (if (and _returnvalue_2 (!= _outguid (CallFunction|GetCurrentReplicatedWeaponGuid)))
                  (CallFunction|DestroyBackpack _outguid))
                (CallFunction|InternalSpawnBackpack _outbackpackclass)
                (else
                  (CallFunction|ClearListenEvents)
                  (Utilities|IsValid (Variables|Default|GetBackpackActorRef)
                    (:Is%20Valid
                      (CallFunction|DestroyBackpack (CallFunction|GetCurrentLocalWeaponGuid)))))))
            (:then_1
              (CallFunction|SetLocalClassandGuid (CallFunction|GetCurrentReplicatedBackpackClass) (CallFunction|GetCurrentReplicatedWeaponGuid))
              (CallFunction|SetBackpackHidden (Variables|SpriteBackpackController|GetbServerHiddenState)))))))))

(fn InternalSpawnBackpack
  (params (Class :type Actor%20Class%20Reference))
  (bind (_outguid _returnvalue) (CallFunction|GetCurrentLocalWeaponGuid))
  (if (CallFunction|IsBackpackEnabled)
    ; Both Local and Replicated Spawn functions should already have destroyed the current backpack actor
    (if (not _returnvalue)
      (bind _returnvalue_2 (Game|SpawnActorfromClass Class (Math|Transform|MakeTransform "0, 0, 0") "AlwaysSpawn" "MultiplyWithRoot" (Variables|Default|GetOwnerPawn) (Variables|Default|GetOwnerPawn)))
      (bind _asb_sprite_held_backpack (Utilities|Casting|CastToB_Sprite_Held_Backpack _returnvalue_2))
      (bind _output_get (Variables|Default|SetBackpackActorRef _asb_sprite_held_backpack))
      (Rendering|SetActorHiddenInGame _output_get false)
      (CallFunction|InitializeBackpack)
      (return _output_get))))

(fn IsBackpackEnabled
  (return (> (Fortnite|GetNumericCVar "Sprites.EnableBackpack") 0)))

(fn IsBackpackReplicationEnabled
  (return (and (> (Fortnite|GetNumericCVar "Sprites.EnableBackpackReplication") 0) (CallFunction|IsBackpackEnabled))))

(fn LocallySpawnBackpack
  (params (Class :type Actor%20Class%20Reference) (ItemGuid :type Guid%20Structure))
  (bind (_outguid _returnvalue) (CallFunction|GetCurrentLocalWeaponGuid))
  (if (and (!= _outguid ItemGuid) _returnvalue)
    (CallFunction|DestroyBackpack (CallFunction|GetCurrentLocalWeaponGuid)))
  (bind _backpackactor (CallFunction|InternalSpawnBackpack Class))
  (CallFunction|SetLocalClassandGuid Class ItemGuid)
  (return _backpackactor))
