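; Player controller component that marks the nearest discoverable, apparently a shrine, when the player picks up a Sprite.
; On authority, a hotfix-gated InInventory tag listener runs CanPingShrine, then triggers discovery and applies a cooldown effect; losing the tag cancels the stored marker.
; The discovery callback marks the result only within a hotfix-set distance.
; Includes a full-ammo check helper and placeholder debug drawing for nearest-discoverable changes.
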(event EventBeginPlay
  ; Bind to Sprite's getting into inventory and triggering discovery manually when that happens so we can use the results. Only triggering when sprites have full ammo.
  (Utilities|SwitchHasAuthority
    (:Authority
      (if (ScalableFloat|GetValueasBool (Variables|Hotfixes|GetHF_ShouldPingDiscoverablesWhenPickingUpSprites))
        (bind (_asynctaskproxy _tagcount) (Ability|Tasks|WaitGameplayTagCountChangedOnActor (Components|GetOwner) "(TagName=\"Athena.Player.Sprite.InInventory\")")
          (:TagCountChanged
            (if (> _tagcount 0)
              (Utilities|CanPingShrine
                (:True
                  (EventDispatchers|BindEventtoOnNearestDiscoverableFoundEvent (EventDispatchers|CreateEvent :!event Custom|OnDiscoverableFound))
                  (CallFunction|TriggerDiscovery)
                  ; Apply a cooldown to the ping
                  (GameplayEffects|ApplyGameplayEffectToSelf :self (Ability|GetFortAbilitySystemComponent (Components|GetOwner)) :GameplayEffectClass "/Sprites/Gameplay/Effects/GE_FirePetal_Sprites_PingCooldown.GE_FirePetal_Sprites_PingCooldown_C" :Level 0.0)))
              (else
                (EventDispatchers|UnbindallEventsfromOnNearestDiscoverableFoundEvent)
                (Class|FortWorldMarker|CancelMarker (MarkerComponent|FindMarkerbyID (Actor|GetComponentbyClass (Components|GetOwner) "/Script/FortniteGame.AthenaMarkerComponent") (Variables|Default|GetMarkedShrineID)))))))))))

(event Custom|OnDiscoverableFound
  (params (Discoverable :type Sprite%20Discoverable%20Interface%20Interface%20%28by%20ref%29))
  ; Ping Nearest Discoverable
  (bind _asactor (Utilities|Casting|CastToActor (Utilities|ToObject(Interface) Discoverable)))
  (EventDispatchers|UnbindallEventsfromOnNearestDiscoverableFoundEvent)
  (Utilities|SwitchHasAuthority
    (:Authority
      (if (<= (Transformation|GetDistanceTo _asactor (Game|Player|GetPlayerPawn (Utilities|Casting|CastToFortPlayerController (Components|GetOwner)))) (ScalableFloat|GetValueatLevel (Variables|Hotfixes|GetHF_MaxDiscoverablePingDistance)))
        (bind _returnvalue (MarkerComponent|MarkActoronServer (Actor|GetComponentbyClass (Components|GetOwner) "/Script/FortniteGame.AthenaMarkerComponent") _asactor false true))
        (Variables|Default|SetMarkedShrineID _returnvalue)))))

(event EventOnNearestDiscoverableChanged
  (params (NewClosest :type Sprite%20Discoverable%20Interface%20Interface%20%28by%20ref%29))
  ; Debug: Please replace with the real implementation
  (bind _asactor (Utilities|Casting|CastToActor (Utilities|ToObject(Interface) NewClosest)))
  (Rendering|Debug|DrawDebugSphere (+ (Transformation|GetActorLocation _asactor) "0.000000,0.000000,700.000000") 1e+02 12 "(R=1.000000,G=1.000000,B=1.000000,A=1.000000)" 1e+01 1e+01)
  (Development|PrintString (Utilities|String|Append "New closest at Location: " (Utilities|String|ToString(Vector) (Transformation|GetActorLocation _asactor)))))

(fn DoesEquippedSpriteHaveFullAmmo
  (bind _returnvalue (Inventory|GetInventoryItemwithTag (Utilities|Casting|CastToFortPlayerController (Components|GetOwner)) (Variables|Default|GetSpriteItemParentTag)))
  (bind _output_get (LocalVariables|SetSpriteItem _returnvalue))
  (bind _outclipsize (WeaponItemDefinition|TryGetBaseClipSizewithMods (Utilities|Casting|CastToFortWeaponRangedItemDefinition (Item|GetItemDefinition _output_get)) (Utilities|Casting|CastToFortWorldItem (LocalVariables|GetSpriteItem))))
  (return (== (Item|GetTotalAmmo (LocalVariables|GetSpriteItem)) _outclipsize)))
