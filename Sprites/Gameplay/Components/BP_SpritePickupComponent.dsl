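; Sprite pickup component that converts its owning actor into an inventory item for a player.
; CanPickup checks remaining uses (loaded ammo), a blocking pawn tag and, under an anti-duping CVar, the PickedUp flag.
; PickupSprite runs on authority, setting PickedUp before calling SpritesUtilityLibrary, then storing the result.
; It passes the owner's item entry or one created from ItemVersion, the currently held sprite item and initial ammo from BP_SpriteTrackingComponent.
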(fn CanPickup
  (params (InteractingPawn :type Fort%20Pawn%20Object%20Reference))
  (Utilities|IsValid InteractingPawn
    (:Is%20Valid
      (return (Math|Boolean|ANDBoolean (CallFunction|HasRemainingUses) (or (not (Variables|Default|GetPickedUp)) (not (> (Fortnite|GetNumericCVar "Fort.Rollback.SpritesDupingPrevention") 0))) (not (GameplayTags|HasMatchingGameplayTag InteractingPawn "(TagName=\"Athena.Player.Sprite.PickupInteraction.Blocked\")")))))))

(fn DropCurrentSprite
  (params (PlayerController :type Fort%20Player%20Controller%20Object%20Reference))
  (Utilities|IsValid PlayerController
    (:Is%20Valid
      (bind _returnvalue (Inventory|GetInventoryItemwithTag PlayerController (Variables|Default|GetSpriteItemTag)))
      (Utilities|IsValid _returnvalue
        (:Is%20Valid
          (return true (Item|GetItemGuid _returnvalue) _returnvalue))))))

(fn HasRemainingUses
  (bind _returnvalue (Components|GetOwner))
  (bind _returnvalue_2 (Actor|GetComponentbyClass _returnvalue "/Script/FortniteGame.FortItemEntryComponent"))
  (Utilities|IsValid _returnvalue
    (:Is%20Valid
      (Utilities|IsValid _returnvalue_2
        (:Is%20Valid
          (if (Class|FortItemEntryComponent|HasItem _returnvalue_2)
            (return (> (Class|FortItemEntryComponent|GetLoadedAmmo _returnvalue_2) 0))
            (else
              _)))
        (:Is%20Not%20Valid
          _))
      (return true))
    (:Is%20Not%20Valid
      _)))

(fn ItemEntryToGive
  (params (Controller :type Fort%20Player%20Controller%20Object%20Reference))
  (bind _returnvalue (Components|GetOwner))
  (bind _fortitementrycomponent (Variables|Default|GetFortItemEntryComponent))
  (Utilities|IsValid _returnvalue
    (:Is%20Valid
      (bind _output_get (Variables|Default|SetFortItemEntryComponent (Actor|GetComponentbyClass _returnvalue "/Script/FortniteGame.FortItemEntryComponent")))
      (Utilities|IsValid _output_get
        (:Is%20Valid
          (if (Class|FortItemEntryComponent|HasItem _fortitementrycomponent)
            (return (Class|FortItemEntryComponent|GetOwnedItem _fortitementrycomponent))
            (else
              _)))
        (:Is%20Not%20Valid
          _)))
    (:Is%20Not%20Valid
      _))
  (return (Loot|CreateItemEntry (Variables|Default|GetItemVersion))))

(fn PickupSprite
  (params (PlayerController :type Fort%20Player%20Controller%20Object%20Reference))
  (bind _returnvalue (Components|GetOwner))
  (bind _returnvalue_2 (Actor|GetComponentbyClass _returnvalue "/Sprites/Gameplay/Components/BP_SpriteTrackingComponent.BP_SpriteTrackingComponent_C"))
  (Networking|SwitchHasAuthority
    (:Authority
      (if (not (Variables|Default|GetPickedUp))
        (Variables|Default|SetPickedUp true)
        (if (and (Utilities|IsValid PlayerController :!class /Script/Engine.KismetSystemLibrary) (Utilities|IsValid (Components|GetOwner) :!class /Script/Engine.KismetSystemLibrary))
          (bind _newitementry (CallFunction|ItemEntrytoGive PlayerController))
          (bind _returnvalue_3 (Inventory|GetInventoryItemwithTag PlayerController (Variables|Default|GetSpriteItemTag)))
          (Utilities|IsValid _returnvalue_2
            (:Is%20Valid
              _)
            (:Is%20Not%20Valid
              _))
          (bind _data (Utilities|FlowControl|MergeExecutionwithData (Variables|/Sprites/Gameplay/Components/BP_SpriteTrackingComponent.BP_SpriteTrackingComponent_C|GetInitial%20Ammo _returnvalue_2) -1)
            (:Then
              (bind _returnvalue_4 (Class|SpritesUtilityLibrary|GiveSpriteItemtoPlayer PlayerController _newitementry (Utilities|Casting|CastToFortWorldItem _returnvalue_3) _data))
              (bind _output_get (Variables|Default|SetPickedUp _returnvalue_4))
              (return _output_get))) ; -1 means "no override"
        )))))
