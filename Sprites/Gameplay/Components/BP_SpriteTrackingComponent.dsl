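; BP_SpriteTrackingComponent holds tracking state for a Sprite: initial ammo, AI spawner utility component and spawner data component list.
; BeginPlay sets initial ammo from an override scalable float, while Initialize sets all three values from its caller.
; TransferTrackingOwnership copies that state onto a valid next tracking component, then fires OnComponentTransfer so ownership can pass between components.
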(event EventBeginPlay
  (|CallParentFunction)
  (Variables|Default|SetInitial%20Ammo (ScalableFloat|GetValueasInteger (Variables|Default|GetInitialAmmoOverrideSF))))

(event EventTick
  (params (DeltaSeconds :type Float%20%28single-precision%29)))

(fn Initialize
  (params (InitialAmmo :type Integer) (AISpawnerUtilityRef :type Fort%20AISpawner%20Utility%20Component%20Object%20Reference) (AISpawnerDataComponentListRef :type Fort%20Athena%20AISpawner%20Data%20Component%20List%20Object%20Reference))
  (Variables|Default|SetInitial%20Ammo InitialAmmo)
  (Variables|Default|SetAISpawnerUtilityComp AISpawnerUtilityRef)
  (Variables|Default|SetAISpawnerDataComponentList AISpawnerDataComponentListRef))

(fn TransferTrackingOwnership
  (params (NextOwnershipComponent :type BP%20Sprite%20Tracking%20Component%20Object%20Reference))
  (Utilities|IsValid NextOwnershipComponent
    (:Is%20Valid
      (Variables|Default|SetInitial%20Ammo (Variables|Default|GetInitial%20Ammo) NextOwnershipComponent)
      (Variables|Default|SetAISpawnerUtilityComp (Variables|Default|GetAISpawnerUtilityComp) NextOwnershipComponent)
      (Variables|Default|SetAISpawnerDataComponentList (Variables|Default|GetAISpawnerDataComponentList) NextOwnershipComponent)
      (Default|CallOnComponentTransfer self NextOwnershipComponent))))
